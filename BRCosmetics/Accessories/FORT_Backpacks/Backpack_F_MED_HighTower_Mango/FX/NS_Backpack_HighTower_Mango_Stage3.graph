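# Niagara effect for stage 3 of the HighTower Mango backpack cosmetic, lighting a pair of red flares with an opening flash.
# Two short-lived burst emitters flash stretched sprites at the paired points and a white-to-red sprite between them.
# Persistent Flare_L and Flare_M sprites pulse in size on system loop age, with a light renderer on Flare_M.
# SoftGlow adds a steady, faint red halo, and camera offsets layer all sprites.
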NiagaraSystem /BRCosmetics/Accessories/FORT_Backpacks/Backpack_F_MED_HighTower_Mango/FX/NS_Backpack_HighTower_Mango_Stage3
dependencies {
 usedDataInterfaces [&NiagaraDataInterfaceVector2DCurve, &NiagaraDataInterfaceColorCurve, &NiagaraDataInterfaceCurve]
 usedDynamicInputs []
 usedModules [&SystemState]
}
emitters {
 Flare_Burst {
  emitterUpdateScript [
   &EmitterState(Life Cycle Mode=::NewEnumerator1, Loop Behavior=::NewEnumerator1, Loop Duration=1, UseLoopDelay=false)
   &SpawnBurst_Instantaneous(Spawn Count=1, Spawn Time=0.01, Use Spawn Probability=false)
   &SpawnBurst_Instantaneous(Spawn Count=1, Spawn Group=1, Spawn Time=0.01, Use Spawn Probability=false) {module=SpawnBurst_Instantaneous001}
  ]
  particleSpawnScript [
   /Niagara/Modules/Spawn/Initialization/InitializeParticle(
    Color=<rgba 1 1 1 25>
    Lifetime=0.35
    Mesh Scale=<1.3 1.3 1.3>
    Module.Write Color=true
    Module.Write Lifetime=true
    Module.Write Mass=true
    Module.Write Material Random=true
    Module.Write Position=false
    Module.Write Scale=false
    Module.Write SpriteRotation=true
    Module.Write SpriteSize=true
    Position=&SimulationPosition()
    Sprite Rotation=&UniformRangedFloat(Minimum=-1, Maximum=1)
    Sprite Size=&Multiply_Vector2(A=<70 20>, B=&Vector2DFromFloat(Value=2))
   )
   &SystemLocation(MaskSpawnGroup=false, Offset=<2.5 -11 -2>)
   &SystemLocation(MaskSpawnGroup=true, Offset=<-2.5 -11 -2>, Spawn Group Mask=1) {module=SystemLocation001}
  ]
  particleUpdateScript [
   &ParticleState()
   &ScaleSpriteSize(Scale Factor=&Vector2DFromCurve(
     Vector2Curve=DataInterface() {propertyValues {
       XCurve {keys [
         {interpMode=RCIM_Cubic, leaveTangent=6.1314, tangentMode=RCTM_Break, time=0, value=0}
         {interpMode=RCIM_Cubic, tangentMode=RCTM_Break, time=0.15, value=1.2}
         <at 1 0.5>
        ]}
       YCurve {keys [
         {interpMode=RCIM_Cubic, leaveTangent=8.3258, tangentMode=RCTM_Break, time=0, value=0}
         {interpMode=RCIM_Cubic, tangentMode=RCTM_Break, time=0.15, value=1.2}
         <at 1 0>
        ]}
      }}
     ScaleCurve=<1 1>
    ))
   &ScaleColor(
    Scale Alpha=&FloatFromCurve(FloatCurve=DataInterface() {propertyValues {Curve {keys [<at 0.7 1>,<at 1 0>]}}})
    Scale RGB=&MakeVectorFromLinearColorRGB(Color=&ColorFromCurve(ColorCurve=DataInterface() {propertyValues {
        AlphaCurve {keys [<at 0 1 RCIM_Cubic>]}
        BlueCurve {keys [<at -0.0151 0.1>, {arriveTangent=-0.0985, interpMode=RCIM_Cubic, leaveTangent=-0.0985, time=0.1174, value=0}, <at 1 0 RCIM_Cubic>]}
        GreenCurve {keys [<at -0.0151 0.1>, {arriveTangent=-0.0985, interpMode=RCIM_Cubic, leaveTangent=-0.0985, time=0.1174, value=0}, <at 1 0 RCIM_Cubic>]}
        RedCurve {keys [<at -0.0151 2>, {arriveTangent=-1.9702, interpMode=RCIM_Cubic, leaveTangent=-1.9702, time=0.1174, value=1}, <at 1 0 RCIM_Cubic>]}
       }}))
    ScaleA=true
    ScaleRGB=true
   )
   &CameraOffset(Camera Offset Amount=2.5)
   &SolveForcesAndVelocity(Clamp Velocity=false, Limit Acceleration=false)
  ]
  rendererClasses [&NiagaraSpriteRendererProperties]
  renderers [{rendererClass=&NiagaraSpriteRendererProperties}]
 }
 Flare_Burst001 {
  emitterUpdateScript [
   &EmitterState(Life Cycle Mode=::NewEnumerator1, Loop Behavior=::NewEnumerator1, Loop Duration=1, UseLoopDelay=false)
   &SpawnBurst_Instantaneous(Spawn Count=1, Use Spawn Probability=false)
  ]
  particleSpawnScript [
   &InitializeParticle(
    Color=&MakeLinearColorFromVectorAndFloat("Vector (RGB)"=&Multiply_VectorByFloat(Vector=<1 1 1>, Float=3), "Float (Alpha)"=1)
    Lifetime=0.25
    Mesh Scale=<1.3 1.3 1.3>
    Module.Write Color=true
    Module.Write Lifetime=true
    Module.Write Mass=true
    Module.Write Material Random=true
    Module.Write Position=false
    Module.Write Scale=false
    Module.Write SpriteRotation=true
    Module.Write SpriteSize=true
    Position=&SimulationPosition()
    Sprite Rotation=45
    Sprite Size=&Vector2DFromFloat(Value=45)
   )
   &SystemLocation(MaskSpawnGroup=false, Offset=<0 -11 -2>)
  ]
  particleUpdateScript [
   &ParticleState()
   &ScaleColor(
    Scale Alpha=&FloatFromCurve(FloatCurve=DataInterface() {propertyValues {Curve {keys [<at 0.2 1>,<at 1 0>]}}})
    Scale RGB=&MakeVectorFromLinearColorRGB(Color=&ColorFromCurve(ColorCurve=DataInterface() {propertyValues {
        AlphaCurve {keys [<at 0 1 RCIM_Cubic>]}
        BlueCurve {keys [<at -0.0448 20>,<at -0.0082 1>,<at 0.2639 0.3073>,<at 0.7992 0 RCIM_Cubic>]}
        GreenCurve {keys [<at -0.0448 17.6768>,<at -0.0082 1>,<at 0.2639 0.3073>,<at 0.7992 0 RCIM_Cubic>]}
        RedCurve {keys [<at -0.0448 4.0625>,<at -0.0082 1>,<at 0.2639 1>,<at 0.7992 1 RCIM_Cubic>]}
       }}))
    ScaleA=true
    ScaleRGB=true
   )
   &ScaleSpriteSize(Scale Factor=&Vector2DFromFloat(Value=&FloatFromCurve(FloatCurve=DataInterface() {propertyValues {Curve {keys [
          {interpMode=RCIM_Cubic, tangentMode=RCTM_Break, time=0, value=0}
          {interpMode=RCIM_Cubic, tangentMode=RCTM_Break, time=0.15, value=1}
          <at 1 0>
         ]}}})))
   &CameraOffset(Camera Offset Amount=20)
   &SolveForcesAndVelocity(Clamp Velocity=false, Limit Acceleration=false)
  ]
  rendererClasses [&NiagaraSpriteRendererProperties]
  renderers [{rendererClass=&NiagaraSpriteRendererProperties}]
 }
 Flare_L {
  emitterUpdateScript [
   &EmitterState(
    Inactive Response=::NewEnumerator1
    Life Cycle Mode=::NewEnumerator1
    Loop Behavior=::NewEnumerator1
    Loop Duration Mode=::NewEnumerator1
    UseLoopDelay=false
   )
   &SpawnBurst_Instantaneous(Spawn Count=1, Use Spawn Probability=false)
   &SpawnBurst_Instantaneous(Spawn Count=1, Spawn Group=1, Use Spawn Probability=false) {module=SpawnBurst_Instantaneous001}
  ]
  particleSpawnScript [
   &InitializeParticle(
    Color=<rgba 1 0 0 0.25>
    Mesh Scale=<1.3 1.3 1.3>
    Module.Write Color=true
    Module.Write Lifetime=true
    Module.Write Mass=true
    Module.Write Material Random=true
    Module.Write Position=false
    Module.Write Scale=false
    Module.Write SpriteRotation=false
    Module.Write SpriteSize=true
    Position=&SimulationPosition()
    Sprite Size=&Vector2DFromFloat(Value=45)
   )
   &SystemLocation(MaskSpawnGroup=false, Offset=<2.5 -11 -2>)
   &SystemLocation(MaskSpawnGroup=true, Offset=<-2.5 -11 -2>, Spawn Group Mask=1) {module=SystemLocation001}
  ]
  particleUpdateScript [
   &ParticleState(Kill Particles When Lifetime Has Elapsed=false, Let Infinitely Lived Particles Die When Emitter Deactivates=true)
   &ScaleSpriteSize(Scale Factor=&Vector2DFromFloat(Value=&FloatFromCurve(
      FloatCurve=DataInterface() {propertyValues {Curve {keys [<at 0 1>,<at 0.1341 0.726>,<at 0.2151 1.053>,<at 0.4268 0.6591>,<at 0.684 1.1084>,<at 0.8653 0.7314>,<at 0.9922 1>]}}}
      CurveIndex=$System.NormalizedLoopAge
     )))
   &ScaleColor(
    Scale RGB=&MakeVectorFromLinearColorRGB(Color=&ColorFromCurve(ColorCurve=DataInterface() {propertyValues {
        AlphaCurve {keys [<at 0 1 RCIM_Cubic>]}
        BlueCurve {keys [<at 0.2 0 RCIM_Cubic>]}
        GreenCurve {keys [<at 0.2 0 RCIM_Cubic>]}
        RedCurve {keys [<at 0.2 1 RCIM_Cubic>]}
       }}))
    ScaleA=true
    ScaleRGB=true
   )
   &CameraOffset(Camera Offset Amount=4.2)
   &SolveForcesAndVelocity(Clamp Velocity=false, Limit Acceleration=false)
  ]
  rendererClasses [&NiagaraSpriteRendererProperties]
  renderers [{rendererClass=&NiagaraSpriteRendererProperties}]
 }
 Flare_M {
  emitterUpdateScript [
   &EmitterState(
    Inactive Response=::NewEnumerator1
    Life Cycle Mode=::NewEnumerator1
    Loop Behavior=::NewEnumerator1
    Loop Duration Mode=::NewEnumerator1
    UseLoopDelay=false
   )
   &SpawnBurst_Instantaneous(Spawn Count=1, Use Spawn Probability=false)
   &SpawnBurst_Instantaneous(Spawn Count=1, Spawn Group=1, Use Spawn Probability=false) {module=SpawnBurst_Instantaneous001}
  ]
  particleSpawnScript [
   &InitializeParticle(
    Color=<rgba 1 0 0 0.2>
    Mesh Scale=<1.3 1.3 1.3>
    Module.Write Color=true
    Module.Write Lifetime=true
    Module.Write Mass=true
    Module.Write Material Random=true
    Module.Write Position=false
    Module.Write Scale=false
    Module.Write SpriteRotation=false
    Module.Write SpriteSize=true
    Position=&SimulationPosition()
    Sprite Size=&Vector2DFromFloat(Value=20)
   )
   &SystemLocation(MaskSpawnGroup=false, Offset=<2.5 -11 -2>)
   &SystemLocation(MaskSpawnGroup=true, Offset=<-2.5 -11 -2>, Spawn Group Mask=1) {module=SystemLocation001}
  ]
  particleUpdateScript [
   &ParticleState(Kill Particles When Lifetime Has Elapsed=false, Let Infinitely Lived Particles Die When Emitter Deactivates=true)
   &ScaleSpriteSize(Scale Factor=&Vector2DFromFloat(Value=&FloatFromCurve(
      FloatCurve=DataInterface() {propertyValues {Curve {keys [<at 0 1>,<at 0.1341 0.726>,<at 0.2151 1.053>,<at 0.4084 0.5013>,<at 0.684 1.1084>,<at 0.8681 0.6573>,<at 0.9922 1>]}}}
      CurveIndex=$System.NormalizedLoopAge
     )))
   &ScaleColor(
    Scale RGB=&MakeVectorFromLinearColorRGB(Color=&ColorFromCurve(ColorCurve=DataInterface() {propertyValues {
        AlphaCurve {keys [<at 0 1 RCIM_Cubic>]}
        BlueCurve {keys [<at 0.2 0 RCIM_Cubic>]}
        GreenCurve {keys [<at 0.2 0 RCIM_Cubic>]}
        RedCurve {keys [<at 0.2 1 RCIM_Cubic>]}
       }}))
    ScaleA=true
    ScaleRGB=true
   )
   &CameraOffset(Camera Offset Amount=3)
   &SolveForcesAndVelocity(Clamp Velocity=false, Limit Acceleration=false)
  ]
  rendererClasses [&NiagaraSpriteRendererProperties, &NiagaraLightRendererProperties]
  renderers [
   {rendererClass=&NiagaraSpriteRendererProperties}
   {rendererClass=&NiagaraLightRendererProperties, rendererIndex=1}
  ]
 }
 SoftGlow {
  emitterUpdateScript [
   &EmitterState(
    Inactive Response=::NewEnumerator1
    Life Cycle Mode=::NewEnumerator1
    Loop Behavior=::NewEnumerator1
    Loop Duration Mode=::NewEnumerator1
    UseLoopDelay=false
   )
   &SpawnBurst_Instantaneous(Spawn Count=1, Use Spawn Probability=false)
  ]
  particleSpawnScript [
   &InitializeParticle(
    Color=<rgba 1 0 0 0.05>
    Mesh Scale=<1.3 1.3 1.3>
    Module.Write Color=true
    Module.Write Lifetime=true
    Module.Write Mass=true
    Module.Write Material Random=true
    Module.Write Position=false
    Module.Write Scale=false
    Module.Write SpriteRotation=false
    Module.Write SpriteSize=true
    Position=&SimulationPosition()
    Sprite Size=<60 55>
   )
   &SystemLocation(MaskSpawnGroup=false, Offset=<0 -10 -2>)
  ]
  particleUpdateScript [
   &ParticleState(Kill Particles When Lifetime Has Elapsed=false, Let Infinitely Lived Particles Die When Emitter Deactivates=true)
   &ScaleSpriteSize(Scale Factor=&Vector2DFromFloat(Value=0.8))
   &ScaleColor(
    Scale RGB=&MakeVectorFromLinearColorRGB(Color=&ColorFromCurve(ColorCurve=DataInterface() {propertyValues {
        AlphaCurve {keys [<at 0 1 RCIM_Cubic>]}
        BlueCurve {keys [<at 0.2 0 RCIM_Cubic>]}
        GreenCurve {keys [<at 0.2 0 RCIM_Cubic>]}
        RedCurve {keys [<at 0.2 1 RCIM_Cubic>]}
       }}))
    ScaleA=true
    ScaleRGB=true
   )
   &CameraOffset(Camera Offset Amount=5)
   &SolveForcesAndVelocity(Clamp Velocity=false, Limit Acceleration=false)
  ]
  rendererClasses [&NiagaraSpriteRendererProperties]
  renderers [{rendererClass=&NiagaraSpriteRendererProperties}]
 }
}
summary {
 systemName=NS_Backpack_HighTower_Mango_Stage3
 emitters [
  {emitterName=Flare_Burst, rendererClasses [&NiagaraSpriteRendererProperties]}
  {emitterName=Flare_L, rendererClasses [&NiagaraSpriteRendererProperties]}
  {emitterName=Flare_M, rendererClasses [&NiagaraSpriteRendererProperties, &NiagaraLightRendererProperties]}
  {emitterName=SoftGlow, rendererClasses [&NiagaraSpriteRendererProperties]}
  {emitterName=Flare_Burst001, rendererClasses [&NiagaraSpriteRendererProperties]}
 ]
 userVariables [User.Right2ImpactVelocity {type=Vector3f, description="", defaultValue {struct=Vector3f, value=<0 0 0>}}]
}
